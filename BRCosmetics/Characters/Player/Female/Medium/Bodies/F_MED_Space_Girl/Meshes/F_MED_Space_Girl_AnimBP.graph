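# Post-process style AnimBP for the F_MED_Space_Girl cosmetic body, copying the parent mesh's pose and correcting its collar, hip and belt bones.
# Additive collar offsets follow melee, a shared targeting/parachute/emote flag, and upper-arm yaw measured against spine_05.
# A constraint drives the hip corrective bone, and AnimDynamics springs swing four collar belts.
# While targeting, the bottom-right belt swaps simulation for a fixed rotation.
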AnimBlueprint /BRCosmetics/Characters/Player/Female/Medium/Bodies/F_MED_Space_Girl/Meshes/F_MED_Space_Girl_AnimBP
target_skeleton=&/F_MED_Space_Girl_Skeleton
graphs [
 AnimGraph {nodes [
   n1 = Get isSkydiving() -> (isSkydiving:double)
   n2 = "Transform (Modify) Bone - Bone: base_collar_l"(
    ComponentPose=Local To Component(LocalPose=Copy Pose From Mesh() {bUseAttachedParent=true})
    Alpha=@n1
    Translation="-3,0,2"
    Rotation="-4,0,0"
    Scale="1,1,1"
   ) {
    rotationMode=BMM_Additive
    rotationSpace=BCS_BoneSpace
    translationMode=BMM_Additive
    translationSpace=BCS_BoneSpace
    alphaBoolBlend {blendOption=Linear}
    alphaScaleBiasClamp {bInterpResult=true}
    boneToModify {boneName=base_collar_l}
    rotation {pitch=-4}
    translation {x=-3, z=2}
   }
   n3 = "Transform (Modify) Bone - Bone: base_collar_r"(ComponentPose=@n2, Alpha=@n1, Translation="3,0,-1", Rotation="-5,0,0", Scale="1,1,1") {
    rotationMode=BMM_Additive
    rotationSpace=BCS_BoneSpace
    translationMode=BMM_Additive
    translationSpace=BCS_BoneSpace
    alphaBoolBlend {blendOption=Linear}
    alphaScaleBiasClamp {bInterpResult=true, interpSpeedDecreasing=5, interpSpeedIncreasing=5}
    boneToModify {boneName=base_collar_r}
    rotation {pitch=-5}
    translation {x=3, z=-1}
   }
   n4 = Get isswlinging() -> (isswlinging:double)
   n5 = "Transform (Modify) Bone - Bone: base_collar_l"(ComponentPose=@n3, Alpha=@n4, Translation="0,0,0", Rotation="30,0,0", Scale="1,1,1") {
    rotationMode=BMM_Additive
    rotationSpace=BCS_BoneSpace
    translationMode=BMM_Additive
    translationSpace=BCS_BoneSpace
    alphaBoolBlend {blendOption=Linear}
    alphaScaleBiasClamp {bInterpResult=true}
    boneToModify {boneName=base_collar_l}
    rotation {pitch=30}
   }
   n6 = "Transform (Modify) Bone - Bone: base_collar_r"(ComponentPose=@n5, Alpha=@n4, Translation="1,0,0", Rotation="30,0,0", Scale="1,1,1") {
    rotationMode=BMM_Additive
    rotationSpace=BCS_BoneSpace
    translationMode=BMM_Additive
    translationSpace=BCS_BoneSpace
    alphaBoolBlend {blendOption=Linear}
    alphaScaleBiasClamp {bInterpResult=true, interpSpeedDecreasing=5, interpSpeedIncreasing=5}
    boneToModify {boneName=base_collar_r}
    rotation {pitch=30}
    translation {x=1}
   }
   n7 = "Constraint - dyn_hip_corrective (PP)"!(ComponentPose=@n6, Alpha=1, ConstraintWeights_0=0.5, ConstraintWeights_1:float=0.5) -> (Pose) {
    alphaBoolBlend {blendOption=Linear}
    boneToModify {boneName=dyn_hip_corrective}
    constraintSetup [{targetBone {boneName=pelvis}}, {targetBone {boneName=thigh_twist_01_l}}]
    constraintWeights [1,1]
   }
   n8 = "Transform (Modify) Bone - Bone: dyn_collar_r"(
    ComponentPose=@n7
    Alpha=Get arm_up_r() -> (arm_up_r:double)
    Translation="-0.593,0.526,-1.632"
    Rotation="-20.000,0.000,23.682"
    Scale="1,1,1"
   ) {
    rotationMode=BMM_Additive
    rotationSpace=BCS_BoneSpace
    translationMode=BMM_Additive
    translationSpace=BCS_BoneSpace
    alphaBoolBlend {blendOption=Linear}
    alphaScaleBiasClamp {interpSpeedDecreasing=1, interpSpeedIncreasing=1}
    boneToModify {boneName=dyn_collar_r}
    rotation=<rot -20 0 23.682>
    translation=<-0.5932 0.5256 -1.6325>
   }
   n9 = "Transform (Modify) Bone - Bone: dyn_collar_l"(ComponentPose=@n8, Alpha=Get arm_up_l() -> (arm_up_l:double), Translation="-0.651,0.000,2", Rotation="-17.901,0.000,0.000", Scale="1,1,1") {
    rotationMode=BMM_Additive
    rotationSpace=BCS_BoneSpace
    translationMode=BMM_Additive
    translationSpace=BCS_BoneSpace
    alphaBoolBlend {blendOption=Linear}
    alphaScaleBiasClamp {interpSpeedDecreasing=1, interpSpeedIncreasing=1}
    boneToModify {boneName=dyn_collar_l}
    rotation {pitch=-17.901}
    translation {x=-0.651, z=2}
   }
   n10 = Anim Dynamics - Bone: dyn_collar_top_belt_l(ComponentPose=@n9, Alpha=1, ExternalForce="0,0,0") {
    angularDampingOverride=1
    angularSpringConstant=1000
    bAngularSpring=true
    bGravityOverrideInSimSpace=true
    bOverrideAngularDamping=true
    gravityScale=0.35
    simulationSpace=RootRelative
    alphaBoolBlend {blendOption=Linear}
    boundBone {boneName=dyn_collar_top_belt_l}
    physicsBodyDefinitions [{
      sphereCollisionRadius=0
      boundBone {boneName=dyn_collar_top_belt_l}
      boxExtents=<2 10 10>
      constraintSetup {angularLimitsMax=<0 20 0>, angularLimitsMin=<0 -20 0>, angularTarget=<1 0 0>}
      localJointOffset=<-2 0 0>
     }]
   }
   n11 = Anim Dynamics - Bone: dyn_collar_bottom_belt_l(ComponentPose=@n10, Alpha=1, ExternalForce="0,0,0") {
    angularDampingOverride=1
    angularSpringConstant=1000
    bAngularSpring=true
    bGravityOverrideInSimSpace=true
    bOverrideAngularDamping=true
    gravityScale=0.35
    simulationSpace=RootRelative
    alphaBoolBlend {blendOption=Linear}
    boundBone {boneName=dyn_collar_bottom_belt_l}
    physicsBodyDefinitions [{
      sphereCollisionRadius=0
      boundBone {boneName=dyn_collar_bottom_belt_l}
      boxExtents=<2 10 10>
      constraintSetup {angularLimitsMax=<0 20 0>, angularLimitsMin=<0 -20 0>, angularTarget=<1 0 0>}
      localJointOffset=<-2 0 0>
     }]
   }
   n12 = Anim Dynamics - Bone: dyn_collar_top_belt_r(ComponentPose=@n11, Alpha=1, ExternalForce="0,0,0") {
    angularDampingOverride=1
    angularSpringConstant=1000
    bAngularSpring=true
    bGravityOverrideInSimSpace=true
    bOverrideAngularDamping=true
    gravityScale=0.35
    simulationSpace=RootRelative
    alphaBoolBlend {blendOption=Linear}
    boundBone {boneName=dyn_collar_top_belt_r}
    physicsBodyDefinitions [{
      sphereCollisionRadius=0
      boundBone {boneName=dyn_collar_top_belt_r}
      boxExtents=<2 10 10>
      constraintSetup {angularLimitsMax=<0 20 0>, angularLimitsMin=<0 -20 0>, angularTarget=<1 0 0>}
      localJointOffset=<2 0 0>
     }]
   }
   n13 = Anim Dynamics - Bone: dyn_collar_bottom_belt_r(ComponentPose=@n12, Alpha=Get ads_alpha() -> (ads_alpha:double), ExternalForce="0,0,0") {
    angularDampingOverride=1
    angularSpringConstant=1000
    bAngularSpring=true
    bGravityOverrideInSimSpace=true
    bOverrideAngularDamping=true
    gravityScale=0.35
    simulationSpace=RootRelative
    alphaBoolBlend {blendOption=Linear}
    boundBone {boneName=dyn_collar_bottom_belt_r}
    physicsBodyDefinitions [{
      sphereCollisionRadius=0
      boundBone {boneName=dyn_collar_bottom_belt_r}
      boxExtents=<2 10 10>
      constraintSetup {angularLimitsMax=<0 20 0>, angularLimitsMin=<0 -20 0>, angularTarget=<1 0 0>}
      localJointOffset=<2 0 0>
     }]
   }
   n14 = "Transform (Modify) Bone - Bone: dyn_collar_bottom_belt_r"(
    ComponentPose=@n13
    Alpha=Get ads_alpha_invert() -> (ads_alpha_invert:double)
    Translation="0,0,0"
    Rotation="19.527,0.000,0.000"
    Scale="1,1,1"
   ) {
    rotationMode=BMM_Additive
    rotationSpace=BCS_BoneSpace
    alphaBoolBlend {blendOption=Linear}
    boneToModify {boneName=dyn_collar_bottom_belt_r}
    rotation {pitch=19.527}
   }
   Output Pose(Result=Component To Local(ComponentPose=@n14)) {layerGroup=None}
  ]}
 EventGraph {nodes [
   Event BlueprintUpdateAnimation()
   n1 = BreakTransform(InTransform=GetDeltaTransformFromRefPose(self=Get PartSkelMesh() -> (PartSkelMesh:Skeletal Mesh Component Object Reference), BoneName=upperarm_r, BaseName=spine_05))
   > Set arm_up_r(arm_up_r:double=MapRangeClamped(Value=BreakRotator(InRot=@n1.Rotation).Yaw, InRangeA=80, InRangeB=138)) -> (Output_Get:double)
   n2 = BreakTransform(InTransform=GetDeltaTransformFromRefPose(self=Get PartSkelMesh(), BoneName=upperarm_l, BaseName=spine_05))
   > Set arm_up_l(arm_up_l:double=MapRangeClamped(Value=BreakRotator(InRot=@n2.Rotation).Yaw, InRangeA=50, InRangeB=105)) -> (Output_Get:double)
   n3 = Get bIsTargeting()
   > Set ads_alpha(ads_alpha:double=SelectFloat(A=0, B=1, bPickA=@n3)) -> (Output_Get:double)
   > Set ads_alpha_invert(ads_alpha_invert:double="To Float (Boolean)"(InBool=@n3)) -> (Output_Get:double)
   n4 = Get bIsParachuteOpened()
   n5 = Get bIsTargeting()
   n6 = Get bIsPlayingEmote()
   > Set isSkydiving(isSkydiving:double="To Float (Boolean)"(InBool=OR Boolean!(A=@n5, B=@n4, C:bool=@n6) -> (ReturnValue))) -> (Output_Get:double)
   > Set isswlinging(isswlinging:double="To Float (Boolean)"(InBool=Get bIsPlayingMeleeAnim())) -> (Output_Get:double)
   Event BlueprintInitializeAnimation()
   > Set PartSkelMesh(PartSkelMesh:Skeletal Mesh Component Object Reference=GetPartSkeletalMeshComponent()) -> (Output_Get:Skeletal Mesh Component Object Reference)
   Reroute Node()
   n7 = BreakTransform(InTransform=GetDeltaTransformFromRefPose(self=Get PartSkelMesh(), BoneName=upperarm_l))
   > Set clavicle_up_r_alpha(clavicle_up_r_alpha:double=MapRangeClamped(Value=BreakRotator(InRot=@n7.Rotation).Pitch, InRangeA=-30, InRangeB=20)) -> (Output_Get:double)
   TryGetPawnOwner()
  ]}
]
variables [
 ads_alpha {type=number, default=1}
 ads_alpha_invert {type=number}
 arm_up_l {type=number}
 arm_up_r {type=number}
 clavicle_up_r_alpha {type=number}
 isSkydiving {type=number}
 isswlinging {type=number}
 PartSkelMesh {type=SkeletalMeshComponent}
]
