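# Animation Blueprint for the Space Girl face accessory skeleton, adding secondary physics motion to the character's hair.
# It copies the attached parent's pose, additively rotates both third strand bones, then runs Anim Dynamics on top, bangs and back bones, plus two head-relative chains for the side strands.
# The Event Graph derives arm-raise values from upper-arm rotation and a parachute-open float, but the AnimGraph shown reads none of them.
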AnimBlueprint /BRCosmetics/Characters/Player/Female/Medium/Bodies/F_MED_Space_Girl/Meshes/Parts/F_MED_Space_Girl_FaceAcc_AnimBP
target_skeleton=&/F_MED_Space_Girl_FaceAcc_Skeleton
graphs [
 AnimGraph {nodes [
   n1 = "Transform (Modify) Bone - Bone: dyn_hairstrand_03_r"(
    ComponentPose=Local To Component(LocalPose=Copy Pose From Mesh() {bUseAttachedParent=true})
    Alpha=1
    Translation="0,0,0"
    Rotation="44.391,25.686,8.673"
    Scale="1,1,1"
   ) {
    rotationMode=BMM_Additive
    rotationSpace=BCS_BoneSpace
    translationMode=BMM_Additive
    translationSpace=BCS_BoneSpace
    alphaBoolBlend {blendOption=Linear}
    boneToModify {boneName=dyn_hairstrand_03_r}
    rotation=<rot 44.391 25.686 8.673>
   }
   n2 = "Transform (Modify) Bone - Bone: dyn_hairstrand_03_l"(ComponentPose=@n1, Alpha=1, Translation="0,0,0", Rotation="42.802,-14.724,-7.683", Scale="1,1,1") {
    rotationMode=BMM_Additive
    rotationSpace=BCS_BoneSpace
    translationMode=BMM_Additive
    translationSpace=BCS_BoneSpace
    alphaBoolBlend {blendOption=Linear}
    boneToModify {boneName=dyn_hairstrand_03_l}
    rotation=<rot 42.802 -14.724 -7.683>
   }
   n3 = Anim Dynamics - Bone: dyn_hairtop_01(ComponentPose=@n2, Alpha=0.3, ExternalForce="0,0,0") {
    angularDampingOverride=1
    bGravityOverrideInSimSpace=true
    bOverrideAngularDamping=true
    gravityScale=0.35
    alphaBoolBlend {blendOption=Linear}
    boundBone {boneName=dyn_hairtop_01}
    physicsBodyDefinitions [{
      sphereCollisionRadius=0
      boundBone {boneName=dyn_hairtop_01}
      boxExtents=<4 10 10>
      constraintSetup {angularLimitsMax=<0 40 40>, angularLimitsMin=<0 -40 -40>, angularTarget=<0 0 0>}
      localJointOffset=<4 0 0>
     }]
   }
   n4 = Anim Dynamics - Bone: dyn_hairtop_02(ComponentPose=@n3, Alpha=0.3, ExternalForce="0,0,0") {
    angularDampingOverride=1
    bGravityOverrideInSimSpace=true
    bOverrideAngularDamping=true
    gravityScale=0.35
    alphaBoolBlend {blendOption=Linear}
    boundBone {boneName=dyn_hairtop_02}
    physicsBodyDefinitions [{
      sphereCollisionRadius=0
      boundBone {boneName=dyn_hairtop_02}
      boxExtents=<4 10 10>
      constraintSetup {angularLimitsMax=<0 40 40>, angularLimitsMin=<0 -40 -40>, angularTarget=<0 0 0>}
      localJointOffset=<4 0 0>
     }]
   }
   n5 = Anim Dynamics - Bone: dyn_hairtop_03(ComponentPose=@n4, Alpha=0.3, ExternalForce="0,0,0") {
    angularDampingOverride=1
    bGravityOverrideInSimSpace=true
    bOverrideAngularDamping=true
    gravityScale=0.35
    alphaBoolBlend {blendOption=Linear}
    boundBone {boneName=dyn_hairtop_03}
    physicsBodyDefinitions [{
      sphereCollisionRadius=0
      boundBone {boneName=dyn_hairtop_03}
      boxExtents=<4 10 10>
      constraintSetup {angularLimitsMax=<0 40 40>, angularLimitsMin=<0 -40 -40>, angularTarget=<0 0 0>}
      localJointOffset=<4 0 0>
     }]
   }
   n6 = Anim Dynamics - Bone: dyn_hairbangs_01(ComponentPose=@n5, Alpha=0.5, ExternalForce="0,0,0") {
    angularDampingOverride=1
    bGravityOverrideInSimSpace=true
    bOverrideAngularDamping=true
    gravityScale=0.35
    alphaBoolBlend {blendOption=Linear}
    boundBone {boneName=dyn_hairbangs_01}
    physicsBodyDefinitions [{
      sphereCollisionRadius=0
      boundBone {boneName=dyn_hairbangs_01}
      boxExtents=<4 10 10>
      constraintSetup {angularLimitsMax=<0 -10 10>, angularLimitsMin=<0 -40 -10>, angularTarget=<0 0 0>}
      localJointOffset=<4 0 0>
     }]
   }
   n7 = Anim Dynamics - Bone: dyn_hairbangs_02(ComponentPose=@n6, Alpha=0.5, ExternalForce="0,0,0") {
    angularDampingOverride=1
    bGravityOverrideInSimSpace=true
    bOverrideAngularDamping=true
    gravityScale=0.35
    alphaBoolBlend {blendOption=Linear}
    boundBone {boneName=dyn_hairbangs_02}
    physicsBodyDefinitions [{
      sphereCollisionRadius=0
      boundBone {boneName=dyn_hairbangs_02}
      boxExtents=<4 10 10>
      constraintSetup {angularLimitsMax=<0 -10 20>, angularLimitsMin=<0 -40 -20>, angularTarget=<0 0 0>}
      localJointOffset=<4 0 0>
     }]
   }
   n8 = Anim Dynamics - Bone: dyn_hair_back(ComponentPose=@n7, Alpha=0.5, ExternalForce="0,0,0") {
    angularDampingOverride=1
    bGravityOverrideInSimSpace=true
    bOverrideAngularDamping=true
    gravityScale=0.35
    alphaBoolBlend {blendOption=Linear}
    boundBone {boneName=dyn_hair_back}
    physicsBodyDefinitions [{
      sphereCollisionRadius=0
      boundBone {boneName=dyn_hair_back}
      boxExtents=<4 10 10>
      constraintSetup {angularLimitsMax=<0 40 0>, angularLimitsMin=<0 -40 -40>, angularTarget=<0 0 0>}
      localJointOffset=<4 0 0>
     }]
   }
   n9 = "Anim Dynamics - Chain: dyn_hairstrand_02_l -> dyn_hairstrand_04_l"(ComponentPose=@n8, Alpha=0.5, ExternalForce="0,0,0") {
    angularDampingOverride=1
    angularSpringConstant=750
    bChain=true
    bOverrideAngularDamping=true
    bUseGravityOverride=true
    gravityScale=0.35
    alphaBoolBlend {blendOption=Linear}
    boundBone {boneName=dyn_hairstrand_02_l}
    chainEnd {boneName=dyn_hairstrand_04_l}
    componentLinearAccScale=<1 1 1>
    componentLinearVelScale=<1 1 1>
    gravityOverride {z=-980}
    physicsBodyDefinitions [
     {
      sphereCollisionRadius=0
      boundBone {boneName=dyn_hairstrand_02_l}
      boxExtents=<4 10 10>
      constraintSetup {angularLimitsMax=<0 0 40>, angularLimitsMin=<0 0 -80>, angularTarget=<1 0 0>}
      localJointOffset=<4 0 0>
     }
     {
      sphereCollisionRadius=0
      boundBone {boneName=dyn_hairstrand_03_l}
      boxExtents=<4 10 10>
      constraintSetup {angularLimitsMax=<0 0 40>, angularLimitsMin=<0 0 -40>, angularTarget=<1 0 0>}
      localJointOffset=<4 0 0>
     }
     {
      sphereCollisionRadius=0
      boundBone {boneName=dyn_hairstrand_04_l}
      boxExtents=<4 10 10>
      constraintSetup {angularLimitsMax=<0 0 40>, angularLimitsMin=<0 0 0>, angularTarget=<1 0 0>}
      localJointOffset=<4 0 0>
     }
    ]
    relativeSpaceBone {boneName=head}
   }
   n10 = "Anim Dynamics - Chain: dyn_hairstrand_02_r -> dyn_hairstrand_04_r"(ComponentPose=@n9, Alpha=0.5, ExternalForce="0,0,0") {
    angularDampingOverride=1
    angularSpringConstant=750
    bChain=true
    bOverrideAngularDamping=true
    bUseGravityOverride=true
    gravityScale=0.35
    alphaBoolBlend {blendOption=Linear}
    boundBone {boneName=dyn_hairstrand_02_r}
    chainEnd {boneName=dyn_hairstrand_04_r}
    componentLinearAccScale=<1 1 1>
    componentLinearVelScale=<1 1 1>
    gravityOverride {z=-980}
    physicsBodyDefinitions [
     {
      boundBone {boneName=dyn_hairstrand_02_r}
      boxExtents=<10 10 10>
      constraintSetup {angularLimitsMax=<0 0 80>, angularLimitsMin=<0 0 -40>, angularTarget=<0 0 0>}
      localJointOffset=<0 0 0>
     }
     {
      boundBone {boneName=dyn_hairstrand_03_r}
      boxExtents=<10 10 10>
      constraintSetup {angularLimitsMax=<0 0 40>, angularLimitsMin=<0 0 -40>, angularTarget=<0 0 0>}
      localJointOffset=<0 0 0>
     }
     {
      boundBone {boneName=dyn_hairstrand_04_r}
      boxExtents=<10 10 10>
      constraintSetup {angularLimitsMax=<0 0 40>, angularLimitsMin=<0 0 0>, angularTarget=<0 0 0>}
      localJointOffset=<0 0 0>
     }
    ]
    relativeSpaceBone {boneName=head}
   }
   Output Pose(Result=Component To Local(ComponentPose=@n10)) {layerGroup=None}
  ]}
 EventGraph {nodes [
   Event BlueprintUpdateAnimation()
   > Sequence()
   > Is Valid(InputObject=Get body_part() -> (body_part:Skeletal Mesh Component Object Reference))
   n1 = BreakTransform(InTransform=GetDeltaTransformFromRefPose(self=Get body_part(), BoneName=upperarm_r, BaseName=spine_05))
   > Set arm_up_r(arm_up_r:double=MapRangeClamped(Value=BreakRotator(InRot=@n1.Rotation).Yaw, InRangeA=80, InRangeB=138)) -> (Output_Get:double)
   n2 = BreakTransform(InTransform=GetDeltaTransformFromRefPose(self=Get body_part(), BoneName=upperarm_l, BaseName=spine_05))
   > Set arm_up_l(arm_up_l:double=MapRangeClamped(Value=BreakRotator(InRot=@n2.Rotation).Yaw, InRangeA=50, InRangeB=105)) -> (Output_Get:double)
   > Set isSkydiving(isSkydiving:double="To Float (Boolean)"(InBool=Get bIsParachuteOpened())) -> (Output_Get:double)
   Event BlueprintInitializeAnimation()
   > n3 = Cast To FortPlayerPawnAthena(Object=TryGetPawnOwner())
   > Is Valid(InputObject=@n3)
   > Set body_part(body_part:Skeletal Mesh Component Object Reference=GetSkeletalMeshForPartType(self=@n3)) -> (Output_Get:Skeletal Mesh Component Object Reference)
  ]}
]
variables [
 arm_up_l {type=number}
 arm_up_r {type=number}
 body_part {type=SkeletalMeshComponent}
 isSkydiving {type=number}
]
